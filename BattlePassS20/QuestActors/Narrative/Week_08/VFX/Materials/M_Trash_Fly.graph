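# M_Trash_Fly is an unlit, masked Niagara sprite material for flies, likely around trash in a Week 08 narrative quest.
# The red channel of the SubUV-animated fly texture feeds an emissive adjustment function.
# A PreviousFrameSwitch fakes velocity for motion blur, scaling transformed particle position minus a dynamic-parameter previous position by an intensity.
# That previous position uses the X parameter for all three components, leaving Y and Z unused.
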Material /BattlePassS20/QuestActors/Narrative/Week_08/VFX/Materials/M_Trash_Fly
graph {
 properties {bUsedWithNiagaraSprites=true, blendMode=BLEND_Masked, pixelDepthOffsetMode=PDOM_Legacy, shadingModel=MSM_Unlit, usedShadingModels=Unlit}
 nodes [
  MaterialFunctionCall(Emissive=ParticleSubUV!(UVs, Tex, Apply View MipBias) {texture=/Game/Effects/Fort_Effects/Textures/Fly/T_Fly}.R) {materialFunction=&MF_FX_EmissiveAdjustments}
  n1 = TransformPosition(None=ParticlePositionWS()) {transformSourceType=TRANSFORMPOSSOURCE_World}
  n2 = DynamicParameter() {paramNames [Previous Position X, Previous Position Y, Previous Position Z, Motion Blur Intensity]}
  n3 = MaterialFunctionCall(R=@n2.Previous Position X, G=@n2.Previous Position X, B=@n2.Previous Position X) {materialFunction=&AppendMany}
  PreviousFrameSwitch(Current Frame=Constant3Vector(), Previous Frame=Multiply(A=Subtract(A=@n1, B=@n3.RGB), B=@n2.Motion Blur Intensity))
 ]
}
