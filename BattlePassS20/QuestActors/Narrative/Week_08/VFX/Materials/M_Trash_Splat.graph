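# Unlit translucent Niagara sprite material for a trash splat effect in a Battle Pass S20 narrative quest.
# A water splash texture's green channel feeds normalized erosion, with inverted particle alpha setting the erosion amount.
# The eroded mask blends from white to particle color in a lerp.
# Particle color also feeds an FX emissive adjustment function separately.
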Material /BattlePassS20/QuestActors/Narrative/Week_08/VFX/Materials/M_Trash_Splat
graph {
 properties {bUsedWithNiagaraSprites=true, blendMode=BLEND_Translucent, pixelDepthOffsetMode=PDOM_Legacy, shadingModel=MSM_Unlit, usedShadingModels=Unlit}
 nodes [
  n1 = ParticleColor()
  LinearInterpolate(
   B=@n1.RGB
   Alpha=MaterialFunctionCall!(
    Gradient=TextureSample() {samplerType=SAMPLERTYPE_LinearColor, texture=/Game/Effects/Fort_Effects/Textures/Water/T_waterSplash}.G
    Erosion Amount=OneMinus(None=@n1.A)
   ) {materialFunction=/Game/Effects/Fort_Effects/MaterialFunctions/MF_ErosionNormalized}.Result
  ) {constA=1}
  MaterialFunctionCall(Emissive=@n1.RGB) {materialFunction=&MF_FX_EmissiveAdjustments}
 ]
}
