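# Sound cue for a pizza box explosion in the Season 20 Cleaning Up The Mess quest.
# A Random node picks one of two equally weighted explosion waves and feeds a Modulator at its default settings.
# It uses the crystal shard mission attenuation asset and the mission in-game sound class, with a concurrency override that caps simultaneous instances.
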SoundCue /Game/Athena/Sounds/Quests/CleaningUpTheMess/S20_CleaningUpTheMess_Pizza_Explosion_Cue
root=Modulator_1
nodes [
 {id=Modulator_1, children [Random_1]}
 {id=Random_1, children [WavePlayer_11, WavePlayer_12], properties {weights [1,1]}}
 {id=WavePlayer_11, properties {soundWaveAssetPtr=&/Spot_Pizza_Box_Explosion_01}}
 {id=WavePlayer_12, properties {soundWaveAssetPtr=&/Spot_Pizza_Box_Explosion_02}}
]
properties {
 attenuationSettings=/Game/Sounds/Attenuation/Fort_Mission_CrystalShard_Att
 bOverrideConcurrency=true
 maxDistance=2800
 soundClassObject=/Game/Sounds/FortSoundClassesAndModes/Fort_Mission_IngameSounds
 volumeMultiplier=1
 attenuationOverrides {falloffDistance=2600}
 concurrencyOverrides {maxCount=2}
}
